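# Niagara idle effect for the ChewyStaveGerm pickaxe, twinkling short-lived sparkles over the weapon.
# A single Sparkles emitter spawns sprites at random offsets from a baked point array, attached through a socket reader, and draws them with three sprite renderers, with center, corona and flicker flare attributes computed per particle.
# Alpha flickers along curves and fades in after an emitter-age delay.
# Time of day and quality level scale brightness, and spawn rate differs in the front end.
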NiagaraSystem /BRCosmetics/Weapons/FORT_Melee/Pickaxe_ChewyStaveGerm/FX/NS_Pickaxe_ChewyStaveGerm_Idle
dependencies {
 usedDataInterfaces [
  &NiagaraDataInterfaceColorCurve
  &NiagaraDataInterfaceArrayFloat3
  &NiagaraDataInterfaceCurve
  &NiagaraDataInterfaceArrayFloat
  &NiagaraDataInterfaceSocketReader
 ]
 usedDynamicInputs [/Game/Effects/Niagara/DynamicInputs/TimeOfDay/DaySequenceTimeOfDayColor]
 usedModules [
  &self:SystemSpawnScript.SystemScriptSource.SystemScriptGraph.NiagaraNodeAssignment_1.SetVariables_#1
  &SystemState
  /BRCosmetics/Characters/Player/Female/Medium/Bodies/F_MED_PiperShelf_Pearl/FX/TimeOfDayGlobal
  &self:SystemSpawnScript.SystemScriptSource.SystemScriptGraph.NiagaraNodeAssignment_2.SetVariables_#2
 ]
}
emitters {Sparkles {
  emitterUpdateScript [
   &EmitterState()
   &SpawnRate(
    Spawn Probability=0.25
    SpawnRate=&FN_FrontEndFloatModifier(Gameplay=25, Front End=20)
    Use Spawn Probability=true
   )
  ]
  particleSpawnScript [
   &PartitionParticles(Number Of Partitions=12, Partition Mode:ENiagara_PartitionMode_v2=::NewEnumerator5)
   &InitializeParticle(
    AdjustAlpha=false
    AdjustHue=true
    AdjustSaturation=false
    AdjustValue=false
    Color=&LerpLinearColors(
     StartColor=<rgba 1 1 1 1>
     EndColor=&ColorFromCurve(
      ColorCurve=DataInterface() {propertyValues {
        AlphaCurve {keys [<at 0 1 RCIM_Cubic>]}
        BlueCurve {keys [
          <at 0.3001 0.3853 RCIM_Cubic>
          {arriveTangent=1.7848, interpMode=RCIM_Cubic, leaveTangent=1.7848, time=0.3457, value=1}
          <at 0.6445 1>
          <at 0.6875 2>
         ]}
        GreenCurve {keys [<at 0.3001 1 RCIM_Cubic>,<at 0.3457 1 RCIM_Cubic>,<at 0.6445 1>,<at 0.6875 0.8229>]}
        RedCurve {keys [
          <at 0.3001 0.3853 RCIM_Cubic>
          {arriveTangent=1.7848, interpMode=RCIM_Cubic, leaveTangent=1.7848, time=0.3457, value=1}
          <at 0.6445 1>
          <at 0.6875 0.9761>
         ]}
       }}
      CurveIndex=&RandomRangeFloat(Minimum=0, Maximum=1)
     )
     LerpFactor=0
    )
    Color Mode=::NewEnumerator3
    Lifetime Max=0.4
    Lifetime Min=0.3
    Lifetime Mode=::NewEnumerator1
    Sprite Size Max=<25 25>
    Sprite Size Min=<15 15>
    Sprite Size Mode=::NewEnumerator2
    UsePositionOffset=false
    Write Material Random=true
   )
   &CameraOffset(Camera Offset Amount=5) {module=CameraOffset001}
  ]
  particleUpdateScript [
   &ParticleState()
   &SocketLocation(
    SamplePosOffset=true
    Sampled Position Offset=&AddVectorToPosition(
     Position=<0 0 0>
     Vector=&SelectVectorFromArray(
      Array Sampling Mode=::NewEnumerator0
      Direct Array Index=?
      Vector Selection Array=DataInterface() {propertyValues {FloatData [
         <0.7483 3.4233 -60.8735>
         <0.4614 7.1348 -59.8526>
         <-0.6236 38.146 45.8011>
         <-0.7036 34.6959 49.2754>
         <-1.2379 7.7836 49.6381>
         <-1.1793 10.5949 48.7723>
         <-1.2582 6.8374 50.0896>
         <-1.0593 16.9889 50.7525>
         <-1.1028 14.5072 49.0638>
         <-1.283 5.5605 49.9473>
         <-0.8595 26.4051 46.7789>
         <-1.2064 10.5247 56.4561>
         <-0.9989 20.7401 54.9172>
         <-1.22 9.0955 52.0509>
         <-1.1467 13.3106 55.1106>
         <-1.2026 10.2648 53.7577>
         <-1.3073 5.7434 58.2959>
         <-1.2931 5.2696 51.2538>
         <-1.4271 0.0915 60.624>
         <-1.4132 0.5666 59.3007>
         <-1.3952 0.3742 52.7887>
         <-1.4261 -0.4202 57.3146>
         <-1.7058 -14.2945 58.7534>
         <-1.6037 -9.8729 54.4194>
         <-1.6724 -11.9766 62.5038>
         <-1.6041 -9.1452 58.8565>
         <-1.8299 -20.3423 60.0432>
         <-1.6339 -10.7276 58.3816>
         <-1.7165 -15.7234 53.5122>
         <-1.5385 -6.8165 53.0219>
         <-1.7454 -16.6937 56.3832>
         <-1.4124 -1.1673 49.2084>
         <-1.4638 -4.0425 46.6575>
         <-1.4446 -3.2328 45.8584>
         <-1.3741 0.3678 46.3626>
         <-1.4282 -3.1937 41.3918>
         <-1.3964 -1.2597 43.3785>
         <-1.3306 2.5589 46.4033>
         <-1.1877 5.9329 24.1309>
         <-1.1327 8.3432 21.1412>
         <-0.8664 0.9715 27.2562>
         <2.1781 34.9984 44.7985>
         <2.1536 37.16 50.2579>
         <2.2248 33.5115 49.9513>
         <2.2871 29.7985 46.5864>
         <2.3169 28.2686 46.4385>
         <2.5144 18.3577 46.7801>
         <2.2788 30.6166 48.9408>
         <2.3045 28.7058 45.3057>
         <2.2952 29.1408 45.1155>
         <2.4225 22.8857 46.1212>
         <2.5579 17.5569 55.0364>
         <2.6269 14.3295 56.5655>
         <2.8295 4.0206 56.0855>
         <2.6798 11.167 53.6553>
         <2.5942 15.9557 56.4267>
         <2.3895 25.5316 51.9163>
         <2.631 14.1007 56.4409>
         <2.8802 0.9076 52.8153>
         <3.0672 -7.2539 60.3613>
         <3.0127 -4.3309 61.38>
         <2.9959 -4.1822 57.242>
         <3.0672 -7.7854 57.2363>
         <3.1124 -10.46 54.8927>
         <3.0986 -9.1071 58.7887>
         <3.0456 -6.7356 56.9747>
         <3.243 -16.0296 60.9585>
         <3.0832 -8.671 56.7577>
         <3.057 -8.4998 49.9585>
         <3.0313 -6.9725 51.3041>
         <3.2728 -18.2062 56.9626>
         <3.0596 -8.9418 48.0979>
         <3.015 -7.0895 47.2212>
         <2.8752 0.4133 48.2678>
         <2.983 -5.855 43.3353>
         <2.9061 -2.4574 40.4046>
         <2.8924 -1.5033 42.0277>
         <2.8307 2.2369 45.8486>
         <2.7915 3.0422 38.8425>
         <2.9182 -3.48 38.0411>
         <2.7904 2.8868 37.8516>
         <2.7835 3.2366 37.808>
         <2.8235 -0.0391 28.9371>
         <2.7482 4.8878 36.929>
         <2.7306 5.0242 32.5629>
         <2.7326 4.3325 29.2966>
         <2.6491 7.6181 23.29>
         <2.6716 6.4527 23.5927>
         <2.6132 1.6495 33.6306>
         <2.5606 -0.7277 32.4855>
         <2.6523 1.7731 24.1398>
         <1.934 -1.3348 28.1817>
         <1.6826 -2.9346 32.3691>
         <0.8615 -4.9834 36.9577>
         <1.6753 -4.997 38.2274>
         <2.5218 -4.6504 38.5488>
         <0.0845 -2.4313 30.5149>
         <0.1518 -1.2362 27.0052>
         <-0.9477 -6.4579 43.3935>
         <0.135 -9.2905 46.6309>
         <-1.7211 -16.5748 53.6766>
         <-1.7697 -23.7575 58.9225>
         <0.0377 -13.8718 50.2938>
         <-0.1398 -18.3385 53.7823>
         <2.0037 -17.2539 53.1913>
         <0.9522 -20.1591 54.7341>
         <2.8987 -19.7894 55.6796>
         <2.7369 -10.2553 48.3744>
         <0.6887 42.9733 49.2204>
         <1.4216 41.4817 48.4355>
         <1.0694 41.9619 46.1221>
         <0.8893 41.5503 42.6242>
         <-0.4667 38.392 39.4565>
         <0.5262 39.8292 37.0693>
         <1.0344 39.4791 36.4284>
         <-0.7783 10.1363 18.8752>
         <1.0789 6.0056 34.973>
         <-0.4737 5.9795 35.0677>
         <1.6555 8.5308 25.8069>
         <1.5192 5.2021 37.8953>
         <-0.0195 4.2922 41.228>
         <0.0182 3.2158 45.1708>
         <0.5009 6.9548 -63.5731>
         <-0.1386 12.5327 -57.3035>
         <0.4478 13.1272 -60.0372>
         <1.2749 8.5738 -63.0668>
         <1.2545 20.6095 45.8761>
         <0.2053 28.3599 43.9031>
         <-0.0205 7.2015 49.2895>
         <-0.4725 24.3195 44.9317>
         <1.1547 3.8495 50.1428>
         <-0.9842 9.3312 48.7518>
         <2.6281 1.0205 22.9412>
         <0.6843 1.3023 22.8023>
         <-0.2911 -13.5298 64.7999>
         <-1.2601 -21.9593 60.5621>
         <1.1135 -22.6192 61.6508>
         <2.7191 -16.4664 62.7462>
         <1.8532 -12.4972 65.1121>
         <0.8135 -12.5664 66.0178>
         <2.4245 -4.5832 62.5957>
         <2.0542 1.025 61.4027>
         <1.6595 11.805 58.8294>
         <2.3068 16.7396 56.8188>
         <1.1068 9.8148 59.9364>
         <1.2795 28.156 54.8185>
         <1.2299 37.0962 52.4606>
         <-0.0735 23.4536 55.7135>
         <0.4769 13.0275 59.1879>
         <-1.064 12.9236 57.5714>
         <-0.0831 19.4038 56.7951>
         <0.537 0.1293 62.7187>
         <-1.2567 1.0134 60.7538>
         <-1.4043 -0.2011 60.8125>
         <0.6999 37.421 32.6619>
         <1.3414 37.3667 35.3482>
         <0.885 36.7244 35.4021>
         <0.2303 37.407 34.7628>
         <2.0591 35.7712 41.8031>
        ]}}
      Random Seed=0
     )
    )
    Socket Reader=$System.Socket Reader
    Socket Sampling Mode=::NewEnumerator3
    VelocitySpeedLimit=false
   )
   &ScaleSpriteSize(Uniform Curve Scale=&FNTimeOfDayScalar(Night Scale=1.25, Morning Scale=1, Evening Scale=1))
   &ScaleColor(
    Scale Alpha=&Multiply_Float(
     A=&FloatFromCurve(
      FloatCurve=DataInterface() {propertyValues {Curve {keys [
          {interpMode=RCIM_Cubic, tangentMode=RCTM_User, time=0, value=0}
          {arriveTangent=-2.4058, interpMode=RCIM_Cubic, leaveTangent=-2.4058, tangentMode=RCTM_User, time=0.26, value=0.8421}
          {arriveTangent=-1.1775, interpMode=RCIM_Cubic, leaveTangent=-1.1775, time=0.4613, value=0.0645}
          {arriveTangent=-0.1197, interpMode=RCIM_Cubic, leaveTangent=-0.1197, time=0.5898, value=0.4537}
          <at 1 0 RCIM_Cubic>
         ]}}}
      Scale Curve=&RandomRangeFloat(Minimum=0.1, Maximum=0.5)
     )
     B=&FloatFromCurve(
      FloatCurve=DataInterface() {propertyValues {Curve {keys [{tangentMode=RCTM_User, time=0, value=0}, {tangentMode=RCTM_User, time=1, value=1}]}}}
      Scale Curve=&FloatFromCurve(
       FloatCurve=DataInterface() {propertyValues {Curve {keys [
           {arriveTangentWeight=0.1667, interpMode=RCIM_Cubic, leaveTangentWeight=0.1667, tangentMode=RCTM_User, time=1.4, value=0}
           {tangentMode=RCTM_User, time=2, value=1}
          ]}}}
       CurveIndex=$Emitter.Age
      )
     )
    )
    Scale RGB=&VectorFromFloat(Value=&Multiply_Float(
      A=&RandomRangeFloat(Minimum=0.7, Maximum=0.9, Evaluation Type=::NewEnumerator1)
      B=&SelectFloatFromArray(Direct Array Index=$Engine.QualityLevel, Float Selection Array=DataInterface() {propertyValues {FloatData [10,5,1,1,1]}})
     ))
    ScaleA=true
    ScaleRGB=true
   ) {module=ScaleColor001}
   &ScaleColor(
    Scale Alpha=&FNTimeOfDayScalar(Night Scale=1.25, Morning Scale=1.25, Day Scale=1.25, Evening Scale=1.25)
    Scale RGB=&MakeVectorFromLinearColorRGB(Color=$System.TOD)
    ScaleA=true
    ScaleRGB=true
   )
   /SetVariables(
    Particles.CenterFlareColor:LinearColor=&MultiplyLinearColor(
     Linear Color=$Particles.Color
     Scale Factor=&MakeVector4_FromLinearColor(LinearColor=&LerpLinearColors(StartColor=<rgba 8.0099 9.8007 10 1>, EndColor=<rgba 3.4218 10 3.4218 1>, LerpFactor=0))
    )
    Particles.CenterFlareSize:Vector2f=&Multiply_Vector2(A=$Particles.SpriteSize, B=&Vector2DFromFloat(Value=0.65))
    Particles.CoronaFlareColor:LinearColor=&MultiplyLinearColor(
     Linear Color=$Particles.Color
     Scale Factor=&MakeVector4_FromLinearColor(LinearColor=&LerpLinearColors(StartColor=<rgba 0.15 0.15 0.15 1>, EndColor=<rgba 0.0729 0.0729 0.0729 1>, LerpFactor=0))
    )
    Particles.CoronaFlareSize:Vector2f=&Multiply_Vector2(A=$Particles.Initial.SpriteSize, B=&Vector2DFromFloat(Value=0.85))
    Particles.FlickerFlareColor:LinearColor=&MultiplyLinearColor(
     Linear Color=$Particles.Color
     Scale Factor=&MakeVector4_FromLinearColor(LinearColor=&LerpLinearColors(StartColor=<rgba 0.9441 0.8079 1 1>, EndColor=<rgba 0.0337 0.0417 0.0337 50>, LerpFactor=0))
    )
   ) {isSetParameters=true}
  ]
  rendererClasses [&NiagaraSpriteRendererProperties]
  renderers [
   {rendererClass=&NiagaraSpriteRendererProperties}
   {rendererClass=&NiagaraSpriteRendererProperties, rendererIndex=1}
   {rendererClass=&NiagaraSpriteRendererProperties, rendererIndex=2}
  ]
 }}
summary {
 systemName=NS_Pickaxe_ChewyStaveGerm_Idle
 emitters [{emitterName=Sparkles, rendererClasses [&NiagaraSpriteRendererProperties]}]
 userVariables [
  User.bIsFrontEnd {type=NiagaraBool}
  User.bIsFrontEndPreview {type=NiagaraBool}
 ]
}
