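# Niagara system for the swing trail effect on the ChewyStaveGerm pickaxe.
# Three ribbon emitters use anim trail modules on bone or socket indices: Trail_Screen and Trail_Side draw two trails each, colored near-white or yellow-green by spawn group, while Trail_Side001 draws a grey one.
# The two side emitters lock their particles during update.
# All trails are short-lived, with curve-faded alpha and color scaled by system time-of-day values.
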NiagaraSystem /BRCosmetics/Weapons/FORT_Melee/Pickaxe_ChewyStaveGerm/FX/NS_Pickaxe_ChewyStaveGerm_Swing
dependencies {
 usedDataInterfaces [&NiagaraDataInterfaceCurve, &NiagaraDataInterfaceSocketReader]
 usedDynamicInputs [/Game/Effects/Niagara/DynamicInputs/TimeOfDay/DaySequenceTimeOfDayColor, &FNTimeOfDayScalar]
 usedModules [
  &self:SystemSpawnScript.SystemScriptSource.SystemScriptGraph.NiagaraNodeAssignment_1.SetVariables_#1
  &SystemState
  &self:SystemSpawnScript.SystemScriptSource.SystemScriptGraph.NiagaraNodeAssignment_4.SetVariables_#2
 ]
}
emitters {
 Trail_Screen {
  emitterUpdateScript [
   &EmitterState()
   &FN_SpawnAnimTrail(AnimTrailWidthMode=::FromFirst, Number of Anim Trails=2)
  ]
  particleSpawnScript [
   &FN_InitializeAnimTrailBoneIndecies()
   &FN_InitializeAnimTrailBoneIndecies(AnimTrail Selection=1) {module=FN_InitializeAnimTrailBoneIndecies001}
   &FN_InitializeAnimTrail(
    AnimTrailWidth=&MaskFloat(Mask Source=0, Mask Parameter=?, Mask=0, Masked Float=0.2, Passthrough Float=0.15)
    Color=&MaskLinearColorBySpawnGroup(Masked Linear Color=<rgba 0.9217 0.9217 0.9217 0.7>, Passthrough Linear Color=<rgba 0.5293 0.7605 0.0667 0.7>)
    Lifetime=0.16
   )
  ]
  particleUpdateScript [
   &FN_UpdateAnimTrail() {module=FN_UpdateAnimTrail001}
   &ScaleColor(
    Scale Mode=::NewEnumerator1
    Scale RGBA=&Multiply_Vector4(
     A=&MakeVector4_FromLinearColor(LinearColor=$System.TODColor)
     B=&MakeVector4FromVectorAndFloat(
      "Vector (RGB)"=<1 1 1>
      "Float (Alpha)"=&FloatFromCurve(
       FloatCurve=DataInterface() {propertyValues {Curve {keys [{interpMode=RCIM_Cubic, leaveTangent=-2.5357, tangentMode=RCTM_Break, time=0, value=0}, <at 0.1 1 RCIM_Cubic>, <at 0.8 0>]}}}
       Scale Curve=$System.TODAlpha
      )
     )
    )
    ScaleRGBA=true
   )
  ]
  rendererClasses [&NiagaraRibbonRendererProperties]
  renderers [{rendererClass=&NiagaraRibbonRendererProperties}]
 }
 Trail_Side {
  emitterUpdateScript [
   &EmitterState()
   &FN_SpawnAnimTrail(Number of Anim Trails=2)
  ]
  particleSpawnScript [
   &FN_InitializeAnimTrailBoneIndecies()
   &FN_InitializeAnimTrailBoneIndecies(AnimTrail Selection=1, Second Bone or Socket Index=2) {module=FN_InitializeAnimTrailBoneIndecies001}
   &FN_InitializeAnimTrail(
    Color=&MaskLinearColorBySpawnGroup(Masked Linear Color=<rgba 0.9217 0.9217 0.9217 0.7>, Passthrough Linear Color=<rgba 0.5293 0.7605 0.0667 0.7>)
    Lifetime=0.16
   )
  ]
  particleUpdateScript [
   &FN_UpdateAnimTrail(Lock Particles=true) {module=FN_UpdateAnimTrail001}
   &ScaleColor(
    Scale Alpha=&FloatFromCurve(
     FloatCurve=DataInterface() {propertyValues {Curve {keys [{interpMode=RCIM_Cubic, leaveTangent=-2.5357, tangentMode=RCTM_Break, time=0, value=0}, <at 0.1 1 RCIM_Cubic>, <at 0.8 0>]}}}
     Scale Curve=$System.TODAlpha
    )
    Scale RGB=&MakeVectorFromLinearColorRGB(Color=$System.TODColor)
    ScaleA=true
    ScaleRGB=true
   )
  ]
  rendererClasses [&NiagaraRibbonRendererProperties]
  renderers [{rendererClass=&NiagaraRibbonRendererProperties}]
 }
 Trail_Side001 {
  emitterUpdateScript [
   &EmitterState()
   &FN_SpawnAnimTrail(AnimTrailWidthMode=::FromFirst)
  ]
  particleSpawnScript [
   &FN_InitializeAnimTrailBoneIndecies()
   &FN_InitializeAnimTrail(AnimTrailWidth=0.2, Color=<rgba 0.572 0.6064 0.6163 0.7>, Lifetime=0.16)
  ]
  particleUpdateScript [
   &FN_UpdateAnimTrail(Lock Particles=true) {module=FN_UpdateAnimTrail001}
   &ScaleColor(
    Scale Alpha=&FloatFromCurve(
     FloatCurve=DataInterface() {propertyValues {Curve {keys [{interpMode=RCIM_Cubic, leaveTangent=-2.5357, tangentMode=RCTM_Break, time=0, value=0}, <at 0.1 1 RCIM_Cubic>, <at 0.8 0>]}}}
     Scale Curve=$System.TODAlpha
    )
    Scale RGB=&MakeVectorFromLinearColorRGB(Color=$System.TODColor)
    ScaleA=true
    ScaleRGB=true
   )
  ]
  rendererClasses [&NiagaraRibbonRendererProperties]
  renderers [{rendererClass=&NiagaraRibbonRendererProperties}]
 }
}
summary {
 systemName=NS_Pickaxe_ChewyStaveGerm_Swing
 emitters [
  {emitterName=Trail_Side, rendererClasses [&NiagaraRibbonRendererProperties]}
  {emitterName=Trail_Screen, rendererClasses [&NiagaraRibbonRendererProperties]}
  {emitterName=Trail_Side001, rendererClasses [&NiagaraRibbonRendererProperties]}
 ]
 userVariables [
  User.bIsFrontEnd {type=NiagaraBool}
  User.bIsFrontEndPreview {type=NiagaraBool}
  User.RibbonIDOffset {type=NiagaraInt32}
 ]
}
